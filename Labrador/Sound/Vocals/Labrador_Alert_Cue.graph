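# Sound cue for the Labrador creature's alert vocalization, built from two "GoodBye" voice recordings.
# A Random node picks between the two equally weighted waves, and a Modulator varies pitch and volume on each play.
# It plays at boosted volume through the tactical creature vocalization sound class, using Labrador prey attenuation and a shared prey vocal concurrency set.
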SoundCue /Labrador/Sound/Vocals/Labrador_Alert_Cue
root=Modulator_2
nodes [
 {id=Modulator_2, children [Random_1], properties {pitchMax=1.1, volumeMax=1.1, volumeMin=0.9}}
 {id=Random_1, children [WavePlayer_4,WavePlayer_5], properties {weights [1,1]}}
 {id=WavePlayer_4, properties {soundWaveAssetPtr=&/Labrador_GoodBye_01}}
 {id=WavePlayer_5, properties {soundWaveAssetPtr=&/Labrador_GoodBye_02}}
]
properties {
 attenuationSettings=/Irwin/Sounds/Settings/Prey_Vocalization_Labrador_Att
 maxDistance=6800
 soundClassObject=&Fort_Creatures_Vocalization_Tactical
 volumeMultiplier=2
 concurrencyOverrides {bLimitToOwner=true, maxCount=1, resolutionRule=PreventNew}
 concurrencySet [/Irwin/Sounds/Settings/Conc/Prey_SetGroup_VO_Conc]
}
