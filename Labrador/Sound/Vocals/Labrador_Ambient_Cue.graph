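# Sound cue for Labrador creature ambient vocalizations, using prey vocalization attenuation, a tactical creature sound class and a shared prey voice concurrency set.
# A weighted random node favors an empty branch two to one, so most triggers stay silent.
# The audible branch switches on quality level between an eleven-wave random pool and a four-wave subset.
# A root modulator slightly varies pitch and volume on each play.
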SoundCue /Labrador/Sound/Vocals/Labrador_Ambient_Cue
root=Modulator_1
nodes [
 {id=Modulator_1, children [Random_0], properties {pitchMax=1.1, pitchMin=0.9, volumeMax=1.1, volumeMin=0.9}}
 {id=QualityLevel_1, children [Random_1,Random_1,Random_2,Random_2]}
 {
  id=Random_0
  children [QualityLevel_1, null]
  properties {bRandomizeWithoutReplacement=false, bShouldExcludeFromBranchCulling=true, weights [0.5,1]}
 }
 {
  id=Random_1
  children [
   WavePlayer_2
   WavePlayer_6
   WavePlayer_9
   WavePlayer_19
   WavePlayer_20
   WavePlayer_21
   WavePlayer_22
   WavePlayer_5
   WavePlayer_7
   WavePlayer_8
   WavePlayer_10
  ]
  properties {weights [1,1,1,1,1,1,1,1,1,1,1]}
 }
 {id=Random_2, children [WavePlayer_24, WavePlayer_25, WavePlayer_26, WavePlayer_27], properties {weights [1,1,1,1]}}
 {id=WavePlayer_10, properties {soundWaveAssetPtr=&/Labrador_Ambient_B04}}
 {id=WavePlayer_19, properties {soundWaveAssetPtr=&/Labrador_Ambient_04}}
 {id=WavePlayer_2, properties {soundWaveAssetPtr=&/Labrador_Ambient_01}}
 {id=WavePlayer_20, properties {soundWaveAssetPtr=&/Labrador_Ambient_05}}
 {id=WavePlayer_21, properties {soundWaveAssetPtr=&/Labrador_Ambient_06}}
 {id=WavePlayer_22, properties {soundWaveAssetPtr=&/Labrador_Ambient_07}}
 {id=WavePlayer_24, properties {soundWaveAssetPtr=&Labrador_Ambient_01}}
 {id=WavePlayer_25, properties {soundWaveAssetPtr=&/Labrador_Ambient_02}}
 {id=WavePlayer_26, properties {soundWaveAssetPtr=&/Labrador_Ambient_03}}
 {id=WavePlayer_27, properties {soundWaveAssetPtr=&Labrador_Ambient_04}}
 {id=WavePlayer_5, properties {soundWaveAssetPtr=&/Labrador_Ambient_B01}}
 {id=WavePlayer_6, properties {soundWaveAssetPtr=&Labrador_Ambient_02}}
 {id=WavePlayer_7, properties {soundWaveAssetPtr=&/Labrador_Ambient_B02}}
 {id=WavePlayer_8, properties {soundWaveAssetPtr=&/Labrador_Ambient_B03}}
 {id=WavePlayer_9, properties {soundWaveAssetPtr=&Labrador_Ambient_03}}
]
properties {
 attenuationSettings=/Irwin/Sounds/Settings/Prey_Vocalization_Labrador_Att
 maxDistance=6800
 soundClassObject=&Fort_Creatures_Vocalization_Tactical
 volumeMultiplier=0.8
 concurrencyOverrides {bLimitToOwner=true, maxCount=1, resolutionRule=PreventNew}
 concurrencySet [/Irwin/Sounds/Settings/Conc/Prey_SetGroup_VO_Conc]
}
