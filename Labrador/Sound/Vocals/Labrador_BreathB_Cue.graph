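# Sound cue for the Labrador creature's "B" breath vocalization, using prey vocal attenuation and a tactical creature vocalization sound class.
# A modulator slightly varies pitch and volume over a quality-level switch, atop raised base pitch and lowered volume.
# The first two quality slots randomly pick from eight equally weighted breath waves, the last two from four.
# Concurrency settings include an override allowing one instance per owner, blocking new plays.
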SoundCue /Labrador/Sound/Vocals/Labrador_BreathB_Cue
root=Modulator_1
nodes [
 {id=Modulator_1, children [QualityLevel_1], properties {pitchMax=1.1, pitchMin=0.9, volumeMax=1.1, volumeMin=0.9}}
 {id=QualityLevel_1, children [Random_0,Random_0,Random_3,Random_3]}
 {
  id=Random_0
  children [WavePlayer_0, WavePlayer_1, WavePlayer_2, WavePlayer_12, WavePlayer_13, WavePlayer_14, WavePlayer_15, WavePlayer_16]
  properties {weights [1,1,1,1,1,1,1,1]}
 }
 {id=Random_3, children [WavePlayer_17, WavePlayer_18, WavePlayer_19, WavePlayer_20], properties {weights [1,1,1,1]}}
 {id=WavePlayer_0, properties {soundWaveAssetPtr=&/Labrador_Breath_B01}}
 {id=WavePlayer_1, properties {soundWaveAssetPtr=&/Labrador_Breath_B02}}
 {id=WavePlayer_12, properties {soundWaveAssetPtr=&/Labrador_Breath_B04}}
 {id=WavePlayer_13, properties {soundWaveAssetPtr=&/Labrador_Breath_B05}}
 {id=WavePlayer_14, properties {soundWaveAssetPtr=&/Labrador_Breath_B07}}
 {id=WavePlayer_15, properties {soundWaveAssetPtr=&/Labrador_Breath_B08}}
 {id=WavePlayer_16, properties {soundWaveAssetPtr=&/Labrador_Breath_B09}}
 {id=WavePlayer_17, properties {soundWaveAssetPtr=&Labrador_Breath_B01}}
 {id=WavePlayer_18, properties {soundWaveAssetPtr=&Labrador_Breath_B02}}
 {id=WavePlayer_19, properties {soundWaveAssetPtr=&/Labrador_Breath_B03}}
 {id=WavePlayer_2, properties {soundWaveAssetPtr=&Labrador_Breath_B03}}
 {id=WavePlayer_20, properties {soundWaveAssetPtr=&Labrador_Breath_B04}}
]
properties {
 attenuationSettings=/Irwin/Sounds/Settings/Prey_Vocalization_Att
 maxDistance=6350
 pitchMultiplier=1.4
 soundClassObject=&Fort_Creatures_Vocalization_Tactical
 volumeMultiplier=0.7
 concurrencyOverrides {bLimitToOwner=true, maxCount=1, resolutionRule=PreventNew}
 concurrencySet [null, /Irwin/Sounds/Settings/Conc/Prey_SetGroup_VO_Conc]
}
