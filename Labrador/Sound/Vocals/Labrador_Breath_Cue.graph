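# Sound cue for Labrador breathing vocals, using shared prey vocalization attenuation and concurrency settings and a creature vocalization sound class.
# A quality-level switch feeds the first two levels from a random pool of five breath waves and the last two from a pool of four.
# A root modulator varies pitch and volume slightly, over a raised base pitch and halved volume.
# Concurrency overrides specify one instance per owner, preventing new plays.
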SoundCue /Labrador/Sound/Vocals/Labrador_Breath_Cue
root=Modulator_1
nodes [
 {id=Modulator_1, children [QualityLevel_1], properties {pitchMax=1.1, pitchMin=0.9, volumeMax=1.1, volumeMin=0.9}}
 {id=QualityLevel_1, children [Random_1,Random_1,Random_2,Random_2]}
 {id=Random_1, children [WavePlayer_3,WavePlayer_4,WavePlayer_5,WavePlayer_6,WavePlayer_7], properties {weights [1,1,1,1,1]}}
 {id=Random_2, children [WavePlayer_8, WavePlayer_9, WavePlayer_10, WavePlayer_11], properties {weights [1,1,1,1]}}
 {id=WavePlayer_10, properties {soundWaveAssetPtr=&/Labrador_Breath_03}}
 {id=WavePlayer_11, properties {soundWaveAssetPtr=&/Labrador_Breath_04}}
 {id=WavePlayer_3, properties {soundWaveAssetPtr=&/Labrador_Breath_01}}
 {id=WavePlayer_4, properties {soundWaveAssetPtr=&/Labrador_Breath_02}}
 {id=WavePlayer_5, properties {soundWaveAssetPtr=&Labrador_Breath_03}}
 {id=WavePlayer_6, properties {soundWaveAssetPtr=&Labrador_Breath_04}}
 {id=WavePlayer_7, properties {soundWaveAssetPtr=&/Labrador_Breath_05}}
 {id=WavePlayer_8, properties {soundWaveAssetPtr=&Labrador_Breath_01}}
 {id=WavePlayer_9, properties {soundWaveAssetPtr=&Labrador_Breath_02}}
]
properties {
 attenuationSettings=/Irwin/Sounds/Settings/Prey_Vocalization_Att
 maxDistance=6350
 pitchMultiplier=1.3
 soundClassObject=&Fort_Creatures_Vocalization_Tactical
 volumeMultiplier=0.5
 concurrencyOverrides {bLimitToOwner=true, maxCount=1, resolutionRule=PreventNew}
 concurrencySet [null, /Irwin/Sounds/Settings/Conc/Prey_SetGroup_VO_Conc]
}
