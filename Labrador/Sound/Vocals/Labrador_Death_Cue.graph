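# Sound cue for the Labrador creature's death vocalization, using prey vocalization attenuation and the creature vocalization sound class.
# A Quality Level node sends its first two levels to a random pick of four death waves and the last two to a pick of two.
# A modulator above it slightly varies pitch and boosts volume.
# Concurrency references a lowest-priority death VO set, alongside owner-limited override values that stop the oldest sound.
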SoundCue /Labrador/Sound/Vocals/Labrador_Death_Cue
root=Modulator_1
nodes [
 {id=Modulator_1, children [QualityLevel_1], properties {pitchMax=1.1, pitchMin=1, volumeMax=1.2, volumeMin=1.1}}
 {id=QualityLevel_1, children [Random_2,Random_2,Random_3,Random_3]}
 {id=Random_2, children [WavePlayer_3,WavePlayer_4,WavePlayer_5,WavePlayer_8], properties {weights [1,1,1,1]}}
 {id=Random_3, children [WavePlayer_10, WavePlayer_11], properties {weights [1,1]}}
 {id=WavePlayer_10, properties {soundWaveAssetPtr=&/Labrador_Death_01}}
 {id=WavePlayer_11, properties {soundWaveAssetPtr=&/Labrador_Death_02}}
 {id=WavePlayer_3, properties {soundWaveAssetPtr=&Labrador_Death_01}}
 {id=WavePlayer_4, properties {soundWaveAssetPtr=&Labrador_Death_02}}
 {id=WavePlayer_5, properties {soundWaveAssetPtr=&/Labrador_Death_03}}
 {id=WavePlayer_8, properties {soundWaveAssetPtr=&/Labrador_Death_04}}
]
properties {
 attenuationSettings=/Irwin/Sounds/Settings/Prey_Vocalization_Labrador_Att
 maxDistance=6800
 soundClassObject=/Game/Sounds/FortSoundClassesAndModes/Fort_Creatures_Vocalization_Normal
 volumeMultiplier=1
 concurrencyOverrides {bLimitToOwner=true, maxCount=2, resolutionRule=StopOldest}
 concurrencySet [/Irwin/Sounds/Settings/Conc/SetLowestPriority_VO_Death_Conc]
}
