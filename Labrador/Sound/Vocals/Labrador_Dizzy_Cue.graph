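# Sound cue for the Labrador creature's dizzy vocalization, set up with prey vocal attenuation and concurrency.
# The root random node plays the bark branch about a quarter of the time and otherwise stays silent through an empty child.
# A modulator slightly varies pitch and volume over a quality switch.
# The first two quality slots draw from a weighted four-wave pool, the last two from an even two-wave pool.
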SoundCue /Labrador/Sound/Vocals/Labrador_Dizzy_Cue
root=Random_2
nodes [
 {id=Modulator_0, children [QualityLevel_0], properties {pitchMax=1.1, pitchMin=0.9, volumeMax=1.1, volumeMin=0.9}}
 {id=QualityLevel_0, children [Random_0,Random_0,Random_1,Random_1]}
 {id=Random_0, children [WavePlayer_1,WavePlayer_3,WavePlayer_5,WavePlayer_6], properties {weights [3,1,1,3]}}
 {id=Random_1, children [WavePlayer_19, WavePlayer_20], properties {weights [1,1]}}
 {
  id=Random_2
  children [Modulator_0,null]
  properties {bRandomizeWithoutReplacement=false, bShouldExcludeFromBranchCulling=true, weights [0.5,1.5]}
 }
 {id=WavePlayer_1, properties {soundWaveAssetPtr=&/Labrador_Dizzy_02}}
 {id=WavePlayer_19, properties {soundWaveAssetPtr=&Labrador_Dizzy_02}}
 {id=WavePlayer_20, properties {soundWaveAssetPtr=&/Labrador_Dizzy_03}}
 {id=WavePlayer_3, properties {soundWaveAssetPtr=&Labrador_Dizzy_03}}
 {id=WavePlayer_5, properties {soundWaveAssetPtr=&/Labrador_Dizzy_05}}
 {id=WavePlayer_6, properties {soundWaveAssetPtr=&/Labrador_Dizzy_06}}
]
properties {
 attenuationSettings=/Irwin/Sounds/Settings/Prey_Vocalization_Labrador_Att
 maxDistance=6800
 soundClassObject=&Fort_Creatures_Vocalization_Tactical
 volumeMultiplier=1
 concurrencyOverrides {bLimitToOwner=true, maxCount=1, resolutionRule=PreventNew}
 concurrencySet [/Irwin/Sounds/Settings/Conc/Prey_SetGroup_VO_Conc]
}
