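# Sound cue for the Labrador creature's launch vocalization, classed as a tactical creature vocal.
# A quality-level switch picks five-variation or two-variation branches, each mixing random close waves with quieter random distant waves under separate attenuations.
# A root modulator varies volume and pitch.
# Concurrency overrides specify one instance per owner and block new plays, alongside a shared prey vocal concurrency set.
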SoundCue /Labrador/Sound/Vocals/Labrador_Launch_Cue
root=Modulator_1
nodes [
 {id=Attenuation_0, children [Random_1], properties {attenuationSettings=/Irwin/Sounds/Settings/Predator_Vocalization_Howl_Dist_Att}}
 {id=Attenuation_1, children [Random_0], properties {attenuationSettings=/Irwin/Sounds/Settings/Predator_Vocalization_Att}}
 {id=Attenuation_2, children [Random_7], properties {attenuationSettings=&Predator_Vocalization_Howl_Dist_Att}}
 {id=Attenuation_3, children [Random_6], properties {attenuationSettings=&Predator_Vocalization_Att}}
 {id=Mixer_0, children [Attenuation_1, Attenuation_0], properties {inputVolume [1,0.7]}}
 {id=Mixer_1, children [Attenuation_3, Attenuation_2], properties {inputVolume [1,0.7]}}
 {id=Modulator_1, children [QualityLevel_1], properties {pitchMax=1.1, volumeMax=1.1, volumeMin=0.9}}
 {id=QualityLevel_1, children [Mixer_0,Mixer_0,Mixer_1,Mixer_1]}
 {id=Random_0, children [WavePlayer_0,WavePlayer_1,WavePlayer_2,WavePlayer_3,WavePlayer_6], properties {weights [1,1,1,1,1]}}
 {id=Random_1, children [WavePlayer_7, WavePlayer_8, WavePlayer_9, WavePlayer_11, WavePlayer_17], properties {weights [1,1,1,1,1]}}
 {id=Random_6, children [WavePlayer_28, WavePlayer_29], properties {weights [1,1]}}
 {id=Random_7, children [WavePlayer_33, WavePlayer_34], properties {weights [1,1]}}
 {id=WavePlayer_0, properties {soundWaveAssetPtr=&/Labrador_Launch_01}}
 {id=WavePlayer_1, properties {soundWaveAssetPtr=&/Labrador_Launch_02}}
 {id=WavePlayer_11, properties {soundWaveAssetPtr=&/Labrador_Launch_Distant_04}}
 {id=WavePlayer_17, properties {soundWaveAssetPtr=&/Labrador_Launch_Distant_05}}
 {id=WavePlayer_2, properties {soundWaveAssetPtr=&/Labrador_Launch_03}}
 {id=WavePlayer_28, properties {soundWaveAssetPtr=&Labrador_Launch_01}}
 {id=WavePlayer_29, properties {soundWaveAssetPtr=&Labrador_Launch_02}}
 {id=WavePlayer_3, properties {soundWaveAssetPtr=&/Labrador_Launch_04}}
 {id=WavePlayer_33, properties {soundWaveAssetPtr=&/Labrador_Launch_Distant_01}}
 {id=WavePlayer_34, properties {soundWaveAssetPtr=&/Labrador_Launch_Distant_02}}
 {id=WavePlayer_6, properties {soundWaveAssetPtr=&/Labrador_Launch_05}}
 {id=WavePlayer_7, properties {soundWaveAssetPtr=&Labrador_Launch_Distant_01}}
 {id=WavePlayer_8, properties {soundWaveAssetPtr=&Labrador_Launch_Distant_02}}
 {id=WavePlayer_9, properties {soundWaveAssetPtr=&/Labrador_Launch_Distant_03}}
]
properties {
 maxDistance=18000
 soundClassObject=&Fort_Creatures_Vocalization_Tactical
 volumeMultiplier=0.8
 concurrencyOverrides {bLimitToOwner=true, maxCount=1, resolutionRule=PreventNew}
 concurrencySet [/Irwin/Sounds/Settings/Conc/Prey_SetGroup_VO_Conc]
}
