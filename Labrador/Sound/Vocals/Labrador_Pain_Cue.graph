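# Sound Cue for the Labrador creature's pain vocalization, set up with prey attenuation and a creature vocalization sound class.
# A Mixer layers one of five random TakeDamage clips, briefly delayed, with one of three random LootAppear foley clips mixed at higher input volume.
# Modulators vary pitch on both branches and volume on the vocal.
# Two concurrency sets are assigned alongside a max-count override, with raised volume and a set priority.
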SoundCue /Labrador/Sound/Vocals/Labrador_Pain_Cue
root=Mixer_0
nodes [
 {id=Delay_1, children [Modulator_1], properties {delayMax=0.1, delayMin=0.1}}
 {id=Mixer_0, children [Delay_1,Modulator_0], properties {inputVolume [1.3,2]}}
 {id=Modulator_0, children [Random_1], properties {pitchMax=1.2}}
 {id=Modulator_1, children [Random_0], properties {pitchMax=1.2, volumeMax=1.2, volumeMin=1}}
 {id=Random_0, children [WavePlayer_0,WavePlayer_1,WavePlayer_2,WavePlayer_3,WavePlayer_4], properties {weights [1,1,1,1,1]}}
 {id=Random_1, children [WavePlayer_20, WavePlayer_21, WavePlayer_22], properties {weights [1,1,1]}}
 {id=WavePlayer_0, properties {soundWaveAssetPtr=&/Labrador_TakeDamage_01A}}
 {id=WavePlayer_1, properties {soundWaveAssetPtr=&/Labrador_TakeDamage_02A}}
 {id=WavePlayer_2, properties {soundWaveAssetPtr=&/Labrador_TakeDamage_03A}}
 {id=WavePlayer_20, properties {soundWaveAssetPtr=/Labrador/Sound/Foley/Labrador_LootAppear_01}}
 {id=WavePlayer_21, properties {soundWaveAssetPtr=/Labrador/Sound/Foley/Labrador_LootAppear_02}}
 {id=WavePlayer_22, properties {soundWaveAssetPtr=/Labrador/Sound/Foley/Labrador_LootAppear_03}}
 {id=WavePlayer_3, properties {soundWaveAssetPtr=&/Labrador_TakeDamage_04A}}
 {id=WavePlayer_4, properties {soundWaveAssetPtr=&/Labrador_TakeDamage_05A}}
]
properties {
 attenuationSettings=/Irwin/Sounds/Settings/Prey_Vocalization_Labrador_Att
 maxDistance=6800
 priority=0.9
 soundClassObject=&Fort_Creatures_Vocalization_Tactical
 volumeMultiplier=1.2
 concurrencyOverrides {maxCount=3}
 concurrencySet [/Irwin/Sounds/Settings/Conc/Prey_SetGroup_VO_Conc, /Irwin/Sounds/Settings/Conc/SetLowestPriority_VO_Death_Conc]
}
