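# Sound cue for the Labrador prey creature's panic vocalization, built from six Labrador Alert recordings.
# An equally weighted Random node picks one wave, and a Modulator above it randomizes pitch upward and adds slight volume variation.
# Playback uses the tactical creature vocalization sound class, Labrador-specific prey attenuation, and a shared prey group vocal concurrency set.
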SoundCue /Labrador/Sound/Vocals/Labrador_Panic_Cue
root=Modulator_1
nodes [
 {id=Modulator_1, children [Random_1], properties {pitchMax=1.5, pitchMin=1, volumeMax=1.1, volumeMin=1}}
 {
  id=Random_1
  children [WavePlayer_0, WavePlayer_6, WavePlayer_7, WavePlayer_8, WavePlayer_10, WavePlayer_11]
  properties {weights [1,1,1,1,1,1]}
 }
 {id=WavePlayer_0, properties {soundWaveAssetPtr=&/Labrador_Alert_07_A}}
 {id=WavePlayer_10, properties {soundWaveAssetPtr=&/Labrador_Alert_02_A}}
 {id=WavePlayer_11, properties {soundWaveAssetPtr=&/Labrador_Alert_01_A}}
 {id=WavePlayer_6, properties {soundWaveAssetPtr=&/Labrador_Alert_05_A}}
 {id=WavePlayer_7, properties {soundWaveAssetPtr=&/Labrador_Alert_04_A}}
 {id=WavePlayer_8, properties {soundWaveAssetPtr=&/Labrador_Alert_03_A}}
]
properties {
 attenuationSettings=/Irwin/Sounds/Settings/Prey_Vocalization_Labrador_Att
 maxDistance=6800
 soundClassObject=&Fort_Creatures_Vocalization_Tactical
 volumeMultiplier=0.8
 concurrencyOverrides {bLimitToOwner=true, maxCount=1, resolutionRule=PreventNew}
 concurrencySet [/Irwin/Sounds/Settings/Conc/Prey_SetGroup_VO_Conc]
}
