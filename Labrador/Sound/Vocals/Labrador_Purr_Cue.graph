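# Sound cue for the Labrador creature's purr vocalization, using prey attenuation, a tactical creature vocalization sound class and a shared prey concurrency set.
# A quality-level node sends its first two slots to a random pick of six purr waves and its last two to a random pick of three.
# A root modulator slightly varies pitch and volume on each play.
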SoundCue /Labrador/Sound/Vocals/Labrador_Purr_Cue
root=Modulator_0
nodes [
 {id=Modulator_0, children [QualityLevel_1], properties {pitchMax=1.1, pitchMin=0.9, volumeMax=1.1, volumeMin=0.9}}
 {id=QualityLevel_1, children [Random_4,Random_4,Random_5,Random_5]}
 {
  id=Random_4
  children [WavePlayer_5, WavePlayer_6, WavePlayer_8, WavePlayer_13, WavePlayer_14, WavePlayer_15]
  properties {weights [1,1,1,1,1,1]}
 }
 {id=Random_5, children [WavePlayer_16, WavePlayer_17, WavePlayer_18], properties {weights [1,1,1]}}
 {id=WavePlayer_13, properties {soundWaveAssetPtr=&/Labrador_Purr_04}}
 {id=WavePlayer_14, properties {soundWaveAssetPtr=&/Labrador_Purr_05}}
 {id=WavePlayer_15, properties {soundWaveAssetPtr=&/Labrador_Purr_06}}
 {id=WavePlayer_16, properties {soundWaveAssetPtr=&/Labrador_Purr_01}}
 {id=WavePlayer_17, properties {soundWaveAssetPtr=&/Labrador_Purr_02}}
 {id=WavePlayer_18, properties {soundWaveAssetPtr=&/Labrador_Purr_03}}
 {id=WavePlayer_5, properties {soundWaveAssetPtr=&Labrador_Purr_01}}
 {id=WavePlayer_6, properties {soundWaveAssetPtr=&Labrador_Purr_02}}
 {id=WavePlayer_8, properties {soundWaveAssetPtr=&Labrador_Purr_03}}
]
properties {
 attenuationSettings=/Irwin/Sounds/Settings/Prey_Vocalization_Labrador_Att
 maxDistance=6800
 soundClassObject=&Fort_Creatures_Vocalization_Tactical
 volumeMultiplier=0.6
 concurrencyOverrides {bLimitToOwner=true, maxCount=1, resolutionRule=PreventNew}
 concurrencySet [/Irwin/Sounds/Settings/Conc/Prey_SetGroup_VO_Conc]
}
