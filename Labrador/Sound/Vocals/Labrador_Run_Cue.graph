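# Sound cue for the run vocalizations of the Labrador prey creature.
# An outer random node has two empty slots beside the clip pool, so only about one in three triggers is audible.
# The pool picks evenly from six purr and two ambient waves, and a root modulator slightly varies pitch and volume.
# Prey attenuation, tactical creature vocalization class and a one-voice-per-owner, prevent-new concurrency override shape playback.
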SoundCue /Labrador/Sound/Vocals/Labrador_Run_Cue
root=Modulator_0
nodes [
 {id=Modulator_0, children [Random_3], properties {pitchMax=1.1, pitchMin=0.9, volumeMax=1.1, volumeMin=0.9}}
 {
  id=Random_2
  children [WavePlayer_1, WavePlayer_2, WavePlayer_7, WavePlayer_8, WavePlayer_9, WavePlayer_10, WavePlayer_13, WavePlayer_11]
  properties {weights [1,1,1,1,1,1,1,1]}
 }
 {id=Random_3, children [Random_2,null,null], properties {weights [1,1,1]}}
 {id=WavePlayer_1, properties {soundWaveAssetPtr=&/Labrador_Purr_01}}
 {id=WavePlayer_10, properties {soundWaveAssetPtr=&/Labrador_Purr_06}}
 {id=WavePlayer_11, properties {soundWaveAssetPtr=&/Labrador_Ambient_B01}}
 {id=WavePlayer_13, properties {soundWaveAssetPtr=&/Labrador_Ambient_B04}}
 {id=WavePlayer_2, properties {soundWaveAssetPtr=&/Labrador_Purr_02}}
 {id=WavePlayer_7, properties {soundWaveAssetPtr=&/Labrador_Purr_03}}
 {id=WavePlayer_8, properties {soundWaveAssetPtr=&/Labrador_Purr_04}}
 {id=WavePlayer_9, properties {soundWaveAssetPtr=&/Labrador_Purr_05}}
]
properties {
 attenuationSettings=/Irwin/Sounds/Settings/Prey_Vocalization_Labrador_Att
 bOverrideConcurrency=true
 maxDistance=6800
 soundClassObject=&Fort_Creatures_Vocalization_Tactical
 volumeMultiplier=0.8
 concurrencyOverrides {bLimitToOwner=true, maxCount=1, resolutionRule=PreventNew}
 concurrencySet [null, /Irwin/Sounds/Settings/Conc/Prey_SetGroup_VO_Conc]
}
