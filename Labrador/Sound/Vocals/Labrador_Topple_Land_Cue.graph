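# Sound cue for the Labrador creature's topple landing vocal, classed as a tactical creature vocalization with prey attenuation and a shared prey vocal concurrency set.
# A quality level node routes its first two levels to a random pick of five waves and its last two to a smaller random pool of two.
# A modulator at the root slightly varies pitch and volume on each play.
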SoundCue /Labrador/Sound/Vocals/Labrador_Topple_Land_Cue
root=Modulator_0
nodes [
 {id=Modulator_0, children [QualityLevel_0], properties {pitchMax=1.1, pitchMin=0.9, volumeMax=1.1, volumeMin=0.9}}
 {id=QualityLevel_0, children [Random_0,Random_0,Random_1,Random_1]}
 {id=Random_0, children [WavePlayer_0,WavePlayer_1,WavePlayer_3,WavePlayer_4,WavePlayer_5], properties {weights [1,1,1,1,1]}}
 {id=Random_1, children [WavePlayer_6,WavePlayer_8], properties {weights [1,1]}}
 {id=WavePlayer_0, properties {soundWaveAssetPtr=&/Labrador_Topple_Land_01}}
 {id=WavePlayer_1, properties {soundWaveAssetPtr=&/Labrador_Topple_Land_02}}
 {id=WavePlayer_3, properties {soundWaveAssetPtr=&/Labrador_Topple_Land_03}}
 {id=WavePlayer_4, properties {soundWaveAssetPtr=&/Labrador_Topple_Land_04}}
 {id=WavePlayer_5, properties {soundWaveAssetPtr=&/Labrador_Topple_Land_05}}
 {id=WavePlayer_6, properties {soundWaveAssetPtr=&Labrador_Topple_Land_01}}
 {id=WavePlayer_8, properties {soundWaveAssetPtr=&Labrador_Topple_Land_02}}
]
properties {
 attenuationSettings=/Irwin/Sounds/Settings/Prey_Vocalization_Labrador_Att
 maxDistance=6800
 soundClassObject=&Fort_Creatures_Vocalization_Tactical
 volumeMultiplier=0.9
 concurrencyOverrides {bLimitToOwner=true, maxCount=1, resolutionRule=PreventNew}
 concurrencySet [/Irwin/Sounds/Settings/Conc/Prey_SetGroup_VO_Conc]
}
