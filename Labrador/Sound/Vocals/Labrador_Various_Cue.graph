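# Sound cue for miscellaneous "various" vocalizations of the Labrador prey creature, routed to the tactical creature vocalization sound class.
# A quality-level switch feeds the first two slots from a random pool of seven waves and the last two from a smaller pool of three.
# A modulator adds slight pitch and volume variation, with shared prey attenuation and vocal concurrency settings governing playback.
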SoundCue /Labrador/Sound/Vocals/Labrador_Various_Cue
root=Modulator_0
nodes [
 {id=Modulator_0, children [QualityLevel_0], properties {pitchMax=1.1, pitchMin=0.9, volumeMax=1.1, volumeMin=0.9}}
 {id=QualityLevel_0, children [Random_2,Random_2,Random_4,Random_4]}
 {
  id=Random_2
  children [WavePlayer_2, WavePlayer_7, WavePlayer_9, WavePlayer_10, WavePlayer_11, WavePlayer_12, WavePlayer_13]
  properties {weights [1,1,1,1,1,1,1]}
 }
 {id=Random_4, children [WavePlayer_17, WavePlayer_18, WavePlayer_19], properties {weights [1,1,1]}}
 {id=WavePlayer_10, properties {soundWaveAssetPtr=&/Labrador_Various_04}}
 {id=WavePlayer_11, properties {soundWaveAssetPtr=&/Labrador_Various_05}}
 {id=WavePlayer_12, properties {soundWaveAssetPtr=&/Labrador_Various_06}}
 {id=WavePlayer_13, properties {soundWaveAssetPtr=&/Labrador_Various_07}}
 {id=WavePlayer_17, properties {soundWaveAssetPtr=&/Labrador_Various_01}}
 {id=WavePlayer_18, properties {soundWaveAssetPtr=&/Labrador_Various_02}}
 {id=WavePlayer_19, properties {soundWaveAssetPtr=&/Labrador_Various_03}}
 {id=WavePlayer_2, properties {soundWaveAssetPtr=&Labrador_Various_01}}
 {id=WavePlayer_7, properties {soundWaveAssetPtr=&Labrador_Various_02}}
 {id=WavePlayer_9, properties {soundWaveAssetPtr=&Labrador_Various_03}}
]
properties {
 attenuationSettings=/Irwin/Sounds/Settings/Prey_Vocalization_Labrador_Att
 maxDistance=6800
 soundClassObject=&Fort_Creatures_Vocalization_Tactical
 volumeMultiplier=0.7
 concurrencyOverrides {bLimitToOwner=true, maxCount=1, resolutionRule=PreventNew}
 concurrencySet [/Irwin/Sounds/Settings/Conc/Prey_SetGroup_VO_Conc]
}
